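# Sound cue for the IO Guard voice line that, by its name, plays when a guard enters the DBNO (down but not out) state.
# A Random node picks one of four equally weighted wave variations and is excluded from branch culling.
# Playback uses the NPC voice sound class, close-range guard attenuation and the shared guard chatter concurrency, with a slight volume boost.
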SoundCue /IO_Guard/Sounds/VO/IO_Guard_DBNO_Enter_Cue
root=Random_0
nodes [
 {
  id=Random_0
  children [WavePlayer_1,WavePlayer_2,WavePlayer_3,WavePlayer_4]
  properties {bShouldExcludeFromBranchCulling=true, weights [1,1,1,1]}
 }
 {id=WavePlayer_1, properties {soundWaveAssetPtr=&/IO_Guard_DBNO_Enter_013}}
 {id=WavePlayer_2, properties {soundWaveAssetPtr=&/IO_Guard_DBNO_Enter_010}}
 {id=WavePlayer_3, properties {soundWaveAssetPtr=&/IO_Guard_DBNO_Enter_011}}
 {id=WavePlayer_4, properties {soundWaveAssetPtr=&/IO_Guard_DBNO_Enter_012}}
]
properties {
 attenuationSettings=&/IO_Guard_Close_Att
 maxDistance=4000
 soundClassObject=&Fort_VO_NPC
 volumeMultiplier=1.2
 concurrencySet [&/IO_Guard_Chatter_Conc]
}
