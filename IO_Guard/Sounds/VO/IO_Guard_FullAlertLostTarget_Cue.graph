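# Sound cue for the IO Guard voice bark played when a guard on full alert loses its target.
# A Random node picks one of five equally weighted voice line waves and is excluded from branch culling.
# It plays through the NPC voice sound class with close-range guard attenuation, a slight volume boost, and the guard chatter concurrency set.
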SoundCue /IO_Guard/Sounds/VO/IO_Guard_FullAlertLostTarget_Cue
root=Random_1
nodes [
 {
  id=Random_1
  children [WavePlayer_0, WavePlayer_7, WavePlayer_8, WavePlayer_9, WavePlayer_10]
  properties {bShouldExcludeFromBranchCulling=true, weights [1,1,1,1,1]}
 }
 {id=WavePlayer_0, properties {soundWaveAssetPtr=&/IO_Guard_FullAlertLostTarget_014}}
 {id=WavePlayer_10, properties {soundWaveAssetPtr=&/IO_Guard_FullAlertLostTarget_013}}
 {id=WavePlayer_7, properties {soundWaveAssetPtr=&/IO_Guard_FullAlertLostTarget_010}}
 {id=WavePlayer_8, properties {soundWaveAssetPtr=&/IO_Guard_FullAlertLostTarget_011}}
 {id=WavePlayer_9, properties {soundWaveAssetPtr=&/IO_Guard_FullAlertLostTarget_012}}
]
properties {
 attenuationSettings=&/IO_Guard_Close_Att
 maxDistance=4000
 soundClassObject=&Fort_VO_NPC
 volumeMultiplier=1.2
 concurrencySet [&/IO_Guard_Chatter_Conc]
}
