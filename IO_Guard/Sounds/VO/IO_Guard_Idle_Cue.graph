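# Sound cue for the IO Guard NPC's idle voice lines.
# A Random node picks one of five equally weighted idle wave variants and is excluded from branch culling.
# It plays in the NPC VO sound class with a close-range guard attenuation, a slight volume boost and a chatter concurrency group that governs overlapping guard lines.
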SoundCue /IO_Guard/Sounds/VO/IO_Guard_Idle_Cue
root=Random_1
nodes [
 {
  id=Random_1
  children [WavePlayer_0, WavePlayer_1, WavePlayer_11, WavePlayer_12, WavePlayer_13]
  properties {bShouldExcludeFromBranchCulling=true, weights [1,1,1,1,1]}
 }
 {id=WavePlayer_0, properties {soundWaveAssetPtr=&/IO_Guard_Idle_014}}
 {id=WavePlayer_1, properties {soundWaveAssetPtr=&/IO_Guard_Idle_010}}
 {id=WavePlayer_11, properties {soundWaveAssetPtr=&/IO_Guard_Idle_011}}
 {id=WavePlayer_12, properties {soundWaveAssetPtr=&/IO_Guard_Idle_012}}
 {id=WavePlayer_13, properties {soundWaveAssetPtr=&/IO_Guard_Idle_013}}
]
properties {
 attenuationSettings=&/IO_Guard_Close_Att
 maxDistance=4000
 soundClassObject=&Fort_VO_NPC
 volumeMultiplier=1.2
 concurrencySet [&/IO_Guard_Chatter_Conc]
}
