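# Sound cue for IO Guard voice lines that, judging by the name, play when a suspicious guard returns to idle.
# A Random node picks one of seven equally weighted voice waves and is excluded from branch culling.
# The cue uses the NPC voice sound class with close-range guard attenuation and a slight volume boost.
# A guard chatter concurrency set governs overlapping playback.
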SoundCue /IO_Guard/Sounds/VO/IO_Guard_SuspiciousReturnToIdle_Cue
root=Random_1
nodes [
 {
  id=Random_1
  children [WavePlayer_4, WavePlayer_6, WavePlayer_9, WavePlayer_10, WavePlayer_11, WavePlayer_12, WavePlayer_13]
  properties {bShouldExcludeFromBranchCulling=true, weights [1,1,1,1,1,1,1]}
 }
 {id=WavePlayer_10, properties {soundWaveAssetPtr=&/IO_Guard_SuspiciousReturnToIdle_013}}
 {id=WavePlayer_11, properties {soundWaveAssetPtr=&/IO_Guard_SuspiciousReturnToIdle_014}}
 {id=WavePlayer_12, properties {soundWaveAssetPtr=&/IO_Guard_SuspiciousReturnToIdle_015}}
 {id=WavePlayer_13, properties {soundWaveAssetPtr=&/IO_Guard_SuspiciousReturnToIdle_016}}
 {id=WavePlayer_4, properties {soundWaveAssetPtr=&/IO_Guard_SuspiciousReturnToIdle_010}}
 {id=WavePlayer_6, properties {soundWaveAssetPtr=&/IO_Guard_SuspiciousReturnToIdle_011}}
 {id=WavePlayer_9, properties {soundWaveAssetPtr=&/IO_Guard_SuspiciousReturnToIdle_012}}
]
properties {
 attenuationSettings=&/IO_Guard_Close_Att
 maxDistance=4000
 soundClassObject=&Fort_VO_NPC
 volumeMultiplier=1.2
 concurrencySet [&/IO_Guard_Chatter_Conc]
}
